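# Unlit, two-sided translucent Niagara material for the Knight Cat glider's cylinder gradient effect, giving scrolling streaks tinted by particle color.
# Particle RGB passes through an emissive-adjustment function.
# The alpha chain combines panning streak textures with vertical gradient masks and particle alpha, boosts and offsets the result, then multiplies by a panning Sunrise Castle glider mask.
# View-angle and depth fades soften it, with depth fade bypassed on mobile.
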Material /BRCosmetics/Gadgets/Assets/VinderTech_GliderChute/Glider_Knight_Cat_Female/FX/Default/M_KnightCat_Cylinder_Gradient
graph {
 properties {
  bUsedWithNiagaraMeshParticles=true
  bUsedWithNiagaraRibbons=true
  bUsedWithNiagaraSprites=true
  blendMode=BLEND_Translucent
  pixelDepthOffsetMode=PDOM_Legacy
  shadingModel=MSM_Unlit
  twoSided=true
  usedShadingModels=Unlit
 }
 nodes [
  n1 = ParticleColor()
  MaterialFunctionCall(Emissive=@n1.RGB) {materialFunction=&MF_FX_EmissiveAdjustments}
  n2 = TextureSample(UVs=Panner() {speedX=5}) {
   samplerType=SAMPLERTYPE_LinearColor
   texture=/BRCosmetics/Gadgets/Assets/VinderTech_GliderChute/Glider_Sunrise_Castle_Male/FX/T_FX_SunriseCastleMGlider_End
  }
  n3 = TextureSample() {samplerType=SAMPLERTYPE_Masks, texture=/Game/Global/Textures/Default/Masks_and_Gradients/T_VertGradient_MASK_2}
  n4 = TextureSample(UVs=Panner() {speedX=2.5, speedY=2}) {
   samplerType=SAMPLERTYPE_LinearColor
   texture=/BRCosmetics/Gadgets/Assets/VinderTech_GliderChute/Glider_Sunrise_Castle_Male/FX/T_FX_SunriseCastleMale_MotionBlurStreaks
  }
  n5 = Multiply(A=@n3.RGB, B=@n4.B)
  n6 = Multiply(
   A=@n4.B
   B=TextureSample(UVs=Panner() {speedX=2.5, speedY=5}) {texture=/Game/Global/Textures/Default/Masks_and_Gradients/T_VertGradient_02_sRGB}.RGB
  )
  n7 = Add(A=@n4.RGB) {constB=0.2}
  n8 = Multiply(A=@n3.RGB, B=Add(A=Power(Base=@n5), B=Multiply(A=@n6, B=@n7) {constB=0.1}))
  n9 = Add(A=Multiply(A=Multiply(A=@n1.A, B=@n8)) {constB=30}) {constB=-1}
  n10 = DepthFade(FadeDistance=ScalarParameter() {defaultValue=2, parameterName=DF_FD})
  n11 = Constant() {r=1}
  Multiply(
   A=Multiply(A=Multiply(A=@n2.R, B=@n9), B=ShadingPathSwitch(Default=@n10, Deferred=@n10, Mobile=@n11) {desc=Migrated from FeatureLevelSwitch})
   B=MaterialFunctionCall("FadeSensitivity (optional)"=Constant() {r=2.5}) {materialFunction=&FadeBasedOnViewAngle}.Result
  )
  Reroute(None=@n1.RGB)
 ]
}
