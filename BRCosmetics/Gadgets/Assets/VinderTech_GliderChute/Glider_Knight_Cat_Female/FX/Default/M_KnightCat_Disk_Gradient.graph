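# Translucent, unlit, two-sided Niagara material for a gradient disk effect on the Knight Cat female glider.
# Emissive uses a Custom-node three-color gradient along V, blended toward colorC by a mask.
# Opacity combines panning streak masks, vertex-red-offset smoke and noise, UV edge falloffs, depth fade (bypassed on mobile), view-angle fading and particle alpha.
# The Low_Version scalar blends procedural UV masks toward texture-based or simpler values and adjusts brightness.
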Material /BRCosmetics/Gadgets/Assets/VinderTech_GliderChute/Glider_Knight_Cat_Female/FX/Default/M_KnightCat_Disk_Gradient
graph {
 properties {
  bUsedWithNiagaraMeshParticles=true
  bUsedWithNiagaraRibbons=true
  bUsedWithNiagaraSprites=true
  blendMode=BLEND_Translucent
  pixelDepthOffsetMode=PDOM_Legacy
  shadingModel=MSM_Unlit
  twoSided=true
  usedShadingModels=Unlit
 }
 nodes [
  n1 = ScalarParameter() {parameterName=Low_Version}
  n2 = LinearInterpolate(A=Constant() {r=1.5}, B=Constant() {r=0.75}, Alpha=@n1)
  MaterialExpressionVectorParameter_0 = VectorParameter() {parameterName=colorB, defaultValue=<rgba 0 0.5 0.5 1>}
  MaterialExpressionVectorParameter_3 = VectorParameter() {parameterName=colorD, defaultValue=<rgba 0.433 0.4375 0 1>}
  MaterialExpressionVectorParameter_1 = VectorParameter() {parameterName=colorA, defaultValue=<rgba 2 1 0 1>}
  MaterialExpressionConstant_0 = Constant() {r=0.5}
  n3 = ComponentMask(None=TextureCoordinate()) {g=true}
  MaterialExpressionSubtract_10 = Subtract(A=@n3) {constB=-0.5}
  n4 = Custom!(
   A=@MaterialExpressionVectorParameter_0.RGB
   B=@MaterialExpressionVectorParameter_3.RGB
   C=@MaterialExpressionVectorParameter_1.RGB
   Mid=@MaterialExpressionConstant_0
   Alpha=@MaterialExpressionSubtract_10
  ) {
   code="float3 Col;\r\nCol = lerp(C, B, Alpha/Mid)*step(Alpha,Mid);\r\nCol += lerp(B,A,((Alpha-Mid)/(1-Mid)))*step(Mid,Alpha);\r\nreturn Col;"
   description=CustomLerp
   inputs [
    {
     inputName=A
     input {
      expression=&self:MaterialExpressionVectorParameter_0
      mask=1
      maskB=1
      maskG=1
      maskR=1
     }
    }
    {
     inputName=B
     input {
      expression=&self:MaterialExpressionVectorParameter_3
      mask=1
      maskB=1
      maskG=1
      maskR=1
     }
    }
    {
     inputName=C
     input {
      expression=&self:MaterialExpressionVectorParameter_1
      mask=1
      maskB=1
      maskG=1
      maskR=1
     }
    }
    {
     inputName=Mid
     input {expression=&self:MaterialExpressionConstant_0}
    }
    {
     inputName=Alpha
     input {expression=&self:MaterialExpressionSubtract_10}
    }
   ]
  }
  n5 = Subtract(A=ComponentMask(None=TextureCoordinate()) {r=true}) {constB=0.82}
  n6 = TextureSample() {
   samplerType=SAMPLERTYPE_LinearColor
   texture=&/T_Mask_Disk_KnightCat
  }
  n7 = LinearInterpolate(A=Subtract(A=OneMinus(None=@n5)) {constB=0.5}, B=OneMinus(None=@n6.G), Alpha=@n1)
  n8 = Saturate(None=Multiply(A=Multiply(A=@n2, B=@n4) {constA=1.5}, B=Saturate(None=@n7)))
  n9 = VectorParameter() {parameterName=colorC, defaultValue=<rgba 2 0 1 1>}
  n10 = Multiply(A=Subtract(A=@n3) {constB=0.8}) {constB=5}
  n11 = LinearInterpolate(A=Multiply(A=Multiply(A=@n5) {constB=1.25}, B=OneMinus(None=@n10)), B=Multiply(A=@n6.G) {constB=0.75}, Alpha=@n1)
  n12 = Saturate(None=@n11)
  MaterialFunctionCall(Emissive=Multiply(A=LinearInterpolate(A=@n8, B=@n9.RGB, Alpha=@n12))) {materialFunction=&MF_FX_EmissiveAdjustments}
  n13 = ComponentMask(None=TextureCoordinate()) {g=true}
  n14 = OneMinus(None=@n13)
  n15 = Multiply(A=Subtract(A=@n14) {constB=0.05}, B=Subtract(A=@n13) {constB=0.72})
  n16 = Saturate(None=Multiply(A=Add(A=@n15, B=@n15)))
  n17 = Subtract(A=@n14) {constB=0.3}
  n18 = Multiply(A=@n17) {constB=5}
  n19 = Subtract(A=@n13) {constB=0.05}
  n20 = Saturate(None=Multiply(A=Saturate(None=@n18), B=Multiply(A=@n19) {constB=0.1}))
  n21 = Multiply(A=Multiply(A=@n13, B=Add(A=@n16, B=@n20))) {constB=25}
  n22 = Multiply(A=@n14)
  n23 = Saturate(None=Multiply(A=Subtract(A=@n22) {constB=0.3}, B=Multiply(A=@n19) {constB=25}))
  n24 = Multiply(A=Subtract(A=Multiply(A=@n23, B=@n14), B=Saturate(None=@n17))) {constB=0.1}
  n25 = Add(A=Add(A=Saturate(None=@n21), B=Saturate(None=@n24))) {constB=0.2}
  n26 = Saturate(None=Multiply(A=@n11, B=@n25))
  n27 = TextureSample() {
   samplerType=SAMPLERTYPE_LinearColor
   texture=/BRCosmetics/Gadgets/Assets/VinderTech_GliderChute/Glider_Sunrise_Castle_Male/FX/T_FX_SunriseCastleMale_MotionBlurStreaks
  }
  n28 = TextureSample(UVs=Panner() {speedX=-5}) {texture=/Game/Effects/Fort_Effects/BluGlow/ZoomAnimation}
  n29 = TextureSample(UVs=Panner() {speedX=3.5}) {
   samplerType=SAMPLERTYPE_LinearColor
   texture=&T_Mask_Disk_KnightCat
  }
  n30 = Multiply(A=@n28.R) {constB=0.5}
  n31 = Add(A=Add(A=@n6.R, B=Multiply(A=@n27.R, B=@n28.B) {constA=2.5}), B=Multiply(A=@n29.R, B=Multiply(A=@n30, B=@n29.B)) {constA=2.5})
  n32 = TextureCoordinate() {uTiling=0.05, vTiling=0.8}
  n33 = ComponentMask(None=@n32) {r=true}
  n34 = VertexColor()
  n35 = Panner(Coordinate=AppendVector(A=Add(A=@n33, B=@n34.R), B=ComponentMask(None=@n32) {g=true})) {speedX=-0.5}
  n36 = Multiply(A=Power(Base=TextureSample(UVs=@n35) {texture=&T_Inky_Smoke_Tile}.R) {constExponent=5}) {constB=2}
  n37 = TextureCoordinate()
  n38 = TextureCoordinate() {uTiling=0.05, vTiling=0.8}
  n39 = AppendVector(A=Add(A=ComponentMask(None=@n38) {r=true}, B=VertexColor().R), B=ComponentMask(None=@n38) {g=true})
  n40 = TextureSample(UVs=Panner(Coordinate=OneMinus(None=@n39)) {speedX=0.5}) {samplerType=SAMPLERTYPE_Grayscale, texture=&LowResBlurredNoise}
  n41 = Saturate(None=Multiply(A=Power(Base=@n40.R) {constExponent=5}) {constB=60})
  n42 = OneMinus(None=ComponentMask(None=TextureCoordinate()) {g=true})
  n43 = Multiply(
   A=Saturate(None=Power(Base=@n42) {constExponent=3})
   B=TextureSample(UVs=TextureCoordinate() {uTiling=0, vTiling=0.5}) {samplerType=SAMPLERTYPE_Grayscale, texture=&LowResBlurredNoise}.R
  ) {constB=0.5}
  n44 = Saturate(None=Multiply(A=Power(Base=@n42) {constExponent=3}) {constB=4.5})
  n45 = ComponentMask(None=TextureCoordinate()) {g=true}
  n46 = Add(
   A=Multiply(A=ComponentMask(None=@n37) {g=true}) {constB=0.2}
   B=Multiply(A=LinearInterpolate(A=@n41, B=@n43, Alpha=@n44), B=OneMinus(None=@n45))
  )
  n47 = ComponentMask(None=TextureCoordinate()) {r=true}
  n48 = Multiply(A=Power(Base=OneMinus(None=@n47)) {constExponent=1.5}) {constB=3}
  n49 = Multiply(A=ComponentMask(None=TextureCoordinate()) {r=true}) {constB=6}
  n50 = Multiply(A=Multiply(A=Add(A=@n36, B=@n46)), B=Multiply(A=Saturate(None=@n48), B=Saturate(None=@n49)))
  n51 = ComponentMask(None=TextureCoordinate()) {g=true}
  n52 = Saturate(None=Multiply(A=OneMinus(None=@n51)) {constB=10})
  n53 = Power(Base=Multiply(A=Clamp(None=@n31), B=Multiply(A=@n50, B=@n52))) {constExponent=0.2}
  n54 = TextureCoordinate()
  n55 = DepthFade(
   Opacity=LinearInterpolate(A=ComponentMask(None=@n54) {r=true}, Alpha=@n1)
   FadeDistance=ScalarParameter() {defaultValue=20, parameterName=Depth Fade Distance}
  )
  n56 = ShadingPathSwitch(Default=@n55, Deferred=@n55, Mobile=Constant() {r=1}) {desc=Migrated from FeatureLevelSwitch}
  n57 = Constant()
  n58 = Multiply(
   A=@n25
   B=Multiply(
    A=Multiply(A=@n53, B=@n56)
    B=MaterialFunctionCall("FadeSensitivity (optional)"=@n57) {materialFunction=&FadeBasedOnViewAngle}.Result
   )
  ) {constA=1}
  n59 = Abs(None=DotProduct(A=CameraVectorWS(), B=VertexNormalWS()))
  Multiply(A=Multiply(A=Add(A=@n26, B=@n58), B=Power(Base=@n59) {constExponent=0.333}), B=ParticleColor().A)
 ]
}
