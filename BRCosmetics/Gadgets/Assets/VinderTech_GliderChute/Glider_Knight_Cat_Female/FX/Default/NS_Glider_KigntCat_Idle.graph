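# Niagara system for the idle effect of the female Knight Cat glider: glowing disks, cylinders, a light, sparks and drifting glow drops.
# Disks, cylinders, light and timed ghost-blade bursts follow the User skeletal mesh's bones or sockets, while sparks and drops spawn at offsets.
# User variant colors tint the layers, and system Timeline values blend spawn rates, scale and fade-out.
# Front-end modifiers adjust blade delays, offsets and alpha outside gameplay.
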NiagaraSystem /BRCosmetics/Gadgets/Assets/VinderTech_GliderChute/Glider_Knight_Cat_Female/FX/Default/NS_Glider_KigntCat_Idle
dependencies {
 usedDataInterfaces [
  &NiagaraDataInterfaceColorCurve
  &NiagaraDataInterfaceCurve
  &NiagaraDataInterfaceMeshRendererInfo
  &NiagaraDataInterfaceCurlNoise
  &NiagaraDataInterfaceVectorCurve
  &NiagaraDataInterfaceVector2DCurve
  &NiagaraDataInterfaceArrayFloat
  &NiagaraDataInterfaceArrayFloat2
  &NiagaraDataInterfaceArrayFloat3
  &NiagaraDataInterfaceArrayColor
  &NiagaraDataInterfaceArrayInt32
 ]
 usedDynamicInputs [&SetBoolByFloatComparison]
 usedModules [
  &SystemState
  &self:SystemSpawnScript.SystemScriptSource.SystemScriptGraph.NiagaraNodeAssignment_2.SetVariables_#1
  &self:SystemSpawnScript.SystemScriptSource.SystemScriptGraph.NiagaraNodeAssignment_3.SetVariables_#2
  &self:SystemSpawnScript.SystemScriptSource.SystemScriptGraph.NiagaraNodeAssignment_0.SetVariables_#3
  &self:SystemSpawnScript.SystemScriptSource.SystemScriptGraph.NiagaraNodeAssignment_4.SetVariables_#4
  &self:SystemSpawnScript.SystemScriptSource.SystemScriptGraph.NiagaraNodeAssignment_1.SetVariables_#5
  &Timeline
 ]
}
emitters {
 Blades_Ghost_Trail004 {
  emitterUpdateScript [
   &EmitterState(
    Delay First Loop Only=true
    Enable Distance Culling=true
    Life Cycle Mode=::NewEnumerator1
    Loop Behavior=::NewEnumerator2
    Loop Count=15
    Loop Delay=&FN_FrontEndFloatModifier(Gameplay=1.2, Front End=0.2)
    Loop Duration=0
    Module.bUseMaxDistance=true
    Module.bUseMinDistance=false
    Scalability Mode=::NewEnumerator1
    UseLoopDelay=true
   )
   &SpawnBurst_Instantaneous(Use Spawn Probability=false, UseLoopCountLimit=false)
   &SpawnRate(SpawnRate=50, Use Spawn Probability=false) {enabled=false}
  ]
  particleSpawnScript [
   &InitializeParticle(Color=$User.Variant_Disk_ColorA, Lifetime=0.25, Mesh Scale Mode=::NewEnumerator3, UsePositionOffset=false, Write Material Random=true)
   &SkeletalMeshLocation(
    Bone / Socket Index=5
    Bone Sampling Mode=::NewEnumerator4
    Interpolate Mesh Sampling=false
    MaskSpawnGroup=false
    Skeletal Mesh=$User.SkeletalMesh
   )
   &FN_Mesh_Attributes(Color=$User.Variant_Disk_ColorA, Create Color Attribute=true, "Create Normalized Age "=true, Lifetime=0.25, Write to Module Namespace=true)
  ]
  particleUpdateScript [
   &ParticleState()
   &ScaleColor(
    Linear Color Curve=DataInterface() {propertyValues {
      AlphaCurve {keys [<at 0 1>,<at 0.0654 1>,<at 0.1308 0.05>,<at 0.9264 0>]}
      BlueCurve {keys [<at 0.0022 5>,<at 0.1859 1>,<at 0.5423 1>,<at 0.8523 0.5052>]}
      GreenCurve {keys [<at 0.0022 5>,<at 0.1859 1>,<at 0.5423 1>,<at 0.8523 0.5052>]}
      RedCurve {keys [<at 0.0022 5>,<at 0.1859 1>,<at 0.5423 1>,<at 0.8523 0.5052>]}
     }}
    Scale Mode=::NewEnumerator2
   )
   /SetVariables(Particles.FN_Mesh_Attributes.Color:LinearColor=&ScaleLinearColorByCurve(
     InitialValue=$Particles.Initial.FN_Mesh_Attributes.Color
     ScaleCurve=DataInterface() {propertyValues {
       AlphaCurve {keys [<at 0 1>,<at 0.0491 1>,<at 0.4824 0.05>,<at 0.8219 0>]}
       BlueCurve {keys [<at 0.0022 5>,<at 0.1859 1>,<at 0.5423 1>,<at 0.8523 0.5052>]}
       GreenCurve {keys [<at 0.0022 5>,<at 0.1859 1>,<at 0.5423 1>,<at 0.8523 0.5052>]}
       RedCurve {keys [<at 0.0022 5>,<at 0.1859 1>,<at 0.5423 1>,<at 0.8523 0.5052>]}
      }}
     CurveIndex=$Particles.NormalizedAge
    )) {isSetParameters=true, module=SetVariables_#6}
  ]
  rendererClasses [&NiagaraMeshRendererProperties]
  renderers [
   {rendererClass=&NiagaraMeshRendererProperties, rendererIndex=0}
   {rendererClass=&NiagaraMeshRendererProperties, rendererIndex=1}
  ]
 }
 Blades_Ghost_Trail005 {
  emitterUpdateScript [
   &EmitterState(
    Delay First Loop Only=true
    Enable Distance Culling=true
    Life Cycle Mode=::NewEnumerator1
    Loop Behavior=::NewEnumerator2
    Loop Count=15
    Loop Delay=&FN_FrontEndFloatModifier(Gameplay=1, Front End=0)
    Loop Duration=0
    Module.bUseMaxDistance=true
    Module.bUseMinDistance=false
    Scalability Mode=::NewEnumerator1
    UseLoopDelay=true
   )
   &SpawnBurst_Instantaneous(Use Spawn Probability=false, UseLoopCountLimit=false)
   &SpawnRate(SpawnRate=50, Use Spawn Probability=false) {enabled=false}
  ]
  particleSpawnScript [
   &InitializeParticle(Color=$User.Variant_Disk_ColorA, Lifetime=0.25, Mesh Scale Mode=::NewEnumerator3, UsePositionOffset=false, Write Material Random=true)
   &SkeletalMeshLocation(
    Bone / Socket Index=3
    Bone Sampling Mode=::NewEnumerator4
    Interpolate Mesh Sampling=false
    MaskSpawnGroup=false
    Skeletal Mesh=$User.SkeletalMesh
   )
   &FN_Mesh_Attributes(Color=$User.Variant_Disk_ColorA, Create Color Attribute=true, "Create Normalized Age "=true, Lifetime=0.25, Write to Module Namespace=true)
  ]
  particleUpdateScript [
   &ParticleState()
   &ScaleColor(
    Linear Color Curve=DataInterface() {propertyValues {
      AlphaCurve {keys [<at 0 1>,<at 0.0654 1>,<at 0.1308 0.05>,<at 0.9264 0>]}
      BlueCurve {keys [<at 0.0022 5>,<at 0.1859 1>,<at 0.5423 1>,<at 0.8523 0.5052>]}
      GreenCurve {keys [<at 0.0022 5>,<at 0.1859 1>,<at 0.5423 1>,<at 0.8523 0.5052>]}
      RedCurve {keys [<at 0.0022 5>,<at 0.1859 1>,<at 0.5423 1>,<at 0.8523 0.5052>]}
     }}
    Scale Mode=::NewEnumerator2
   )
   /SetVariables(Particles.FN_Mesh_Attributes.Color:LinearColor=&ScaleLinearColorByCurve(
     InitialValue=$Particles.Initial.FN_Mesh_Attributes.Color
     ScaleCurve=DataInterface() {propertyValues {
       AlphaCurve {keys [<at 0 1>,<at 0.0491 1>,<at 0.4824 0.05>,<at 0.8219 0>]}
       BlueCurve {keys [<at 0.0022 5>,<at 0.1859 1>,<at 0.5423 1>,<at 0.8523 0.5052>]}
       GreenCurve {keys [<at 0.0022 5>,<at 0.1859 1>,<at 0.5423 1>,<at 0.8523 0.5052>]}
       RedCurve {keys [<at 0.0022 5>,<at 0.1859 1>,<at 0.5423 1>,<at 0.8523 0.5052>]}
      }}
     CurveIndex=$Particles.NormalizedAge
    )) {isSetParameters=true, module=SetVariables_#6}
  ]
  rendererClasses [&NiagaraMeshRendererProperties]
  renderers [
   {rendererClass=&NiagaraMeshRendererProperties, rendererIndex=0}
   {rendererClass=&NiagaraMeshRendererProperties, rendererIndex=1}
  ]
 }
 Blades_Ghost_Trail006 {
  emitterUpdateScript [
   &EmitterState(
    Delay First Loop Only=true
    Enable Distance Culling=true
    Life Cycle Mode=::NewEnumerator1
    Loop Behavior=::NewEnumerator2
    Loop Count=15
    Loop Delay=&FN_FrontEndFloatModifier(Gameplay=1.25, Front End=0.25)
    Loop Duration=0
    Module.bUseMaxDistance=true
    Module.bUseMinDistance=false
    Scalability Mode=::NewEnumerator1
    UseLoopDelay=true
   )
   &SpawnBurst_Instantaneous(Use Spawn Probability=false, UseLoopCountLimit=false)
   &SpawnRate(SpawnRate=50, Use Spawn Probability=false) {enabled=false}
  ]
  particleSpawnScript [
   &InitializeParticle(Color=$User.Variant_Disk_ColorA, Lifetime=0.25, Mesh Scale Mode=::NewEnumerator3, UsePositionOffset=false, Write Material Random=true)
   &SkeletalMeshLocation(
    Bone / Socket Index=5
    Bone Sampling Mode=::NewEnumerator4
    Interpolate Mesh Sampling=false
    MaskSpawnGroup=false
    Skeletal Mesh=$User.SkeletalMesh
   )
   &FN_Mesh_Attributes(Color=$User.Variant_Disk_ColorA, Create Color Attribute=true, "Create Normalized Age "=true, Lifetime=0.25, Write to Module Namespace=true)
  ]
  particleUpdateScript [
   &ParticleState()
   &ScaleColor(
    Linear Color Curve=DataInterface() {propertyValues {
      AlphaCurve {keys [<at 0 1>,<at 0.0654 1>,<at 0.1308 0.05>,<at 0.9264 0>]}
      BlueCurve {keys [<at 0.0022 5>,<at 0.1859 1>,<at 0.5423 1>,<at 0.8523 0.5052>]}
      GreenCurve {keys [<at 0.0022 5>,<at 0.1859 1>,<at 0.5423 1>,<at 0.8523 0.5052>]}
      RedCurve {keys [<at 0.0022 5>,<at 0.1859 1>,<at 0.5423 1>,<at 0.8523 0.5052>]}
     }}
    Scale Mode=::NewEnumerator2
   )
   /SetVariables(Particles.FN_Mesh_Attributes.Color:LinearColor=&ScaleLinearColorByCurve(
     InitialValue=$Particles.Initial.FN_Mesh_Attributes.Color
     ScaleCurve=DataInterface() {propertyValues {
       AlphaCurve {keys [<at 0 1>,<at 0.0491 1>,<at 0.4824 0.05>,<at 0.8219 0>]}
       BlueCurve {keys [<at 0.0022 5>,<at 0.1859 1>,<at 0.5423 1>,<at 0.8523 0.5052>]}
       GreenCurve {keys [<at 0.0022 5>,<at 0.1859 1>,<at 0.5423 1>,<at 0.8523 0.5052>]}
       RedCurve {keys [<at 0.0022 5>,<at 0.1859 1>,<at 0.5423 1>,<at 0.8523 0.5052>]}
      }}
     CurveIndex=$Particles.NormalizedAge
    )) {isSetParameters=true, module=SetVariables_#6}
  ]
  rendererClasses [&NiagaraMeshRendererProperties]
  renderers [
   {rendererClass=&NiagaraMeshRendererProperties, rendererIndex=0}
   {rendererClass=&NiagaraMeshRendererProperties, rendererIndex=1}
  ]
 }
 Blades_Ghost_Trail007 {
  emitterUpdateScript [
   &EmitterState(
    Delay First Loop Only=true
    Enable Distance Culling=true
    Life Cycle Mode=::NewEnumerator1
    Loop Behavior=::NewEnumerator2
    Loop Count=15
    Loop Delay=&FN_FrontEndFloatModifier(Gameplay=1.4, Front End=0.4)
    Loop Duration=0
    Module.bUseMaxDistance=true
    Module.bUseMinDistance=false
    Scalability Mode=::NewEnumerator1
    UseLoopDelay=true
   )
   &SpawnBurst_Instantaneous(Use Spawn Probability=false, UseLoopCountLimit=false)
   &SpawnRate(SpawnRate=50, Use Spawn Probability=false) {enabled=false}
  ]
  particleSpawnScript [
   &InitializeParticle(Color=$User.Variant_Disk_ColorA, Lifetime=0.25, Mesh Scale Mode=::NewEnumerator3, UsePositionOffset=false, Write Material Random=true)
   &SkeletalMeshLocation(
    Bone / Socket Index=5
    Bone Sampling Mode=::NewEnumerator4
    Interpolate Mesh Sampling=false
    MaskSpawnGroup=false
    Skeletal Mesh=$User.SkeletalMesh
   )
   &FN_Mesh_Attributes(Color=$User.Variant_Disk_ColorA, Create Color Attribute=true, "Create Normalized Age "=true, Lifetime=0.25, Write to Module Namespace=true)
  ]
  particleUpdateScript [
   &ParticleState()
   &ScaleColor(
    Linear Color Curve=DataInterface() {propertyValues {
      AlphaCurve {keys [<at 0 1>,<at 0.0654 1>,<at 0.1308 0.05>,<at 0.9264 0>]}
      BlueCurve {keys [<at 0.0022 5>,<at 0.1859 1>,<at 0.5423 1>,<at 0.8523 0.5052>]}
      GreenCurve {keys [<at 0.0022 5>,<at 0.1859 1>,<at 0.5423 1>,<at 0.8523 0.5052>]}
      RedCurve {keys [<at 0.0022 5>,<at 0.1859 1>,<at 0.5423 1>,<at 0.8523 0.5052>]}
     }}
    Scale Mode=::NewEnumerator2
   )
   /SetVariables(Particles.FN_Mesh_Attributes.Color:LinearColor=&ScaleLinearColorByCurve(
     InitialValue=$Particles.Initial.FN_Mesh_Attributes.Color
     ScaleCurve=DataInterface() {propertyValues {
       AlphaCurve {keys [<at 0 1>,<at 0.0491 1>,<at 0.4824 0.05>,<at 0.8219 0>]}
       BlueCurve {keys [<at 0.0022 5>,<at 0.1859 1>,<at 0.5423 1>,<at 0.8523 0.5052>]}
       GreenCurve {keys [<at 0.0022 5>,<at 0.1859 1>,<at 0.5423 1>,<at 0.8523 0.5052>]}
       RedCurve {keys [<at 0.0022 5>,<at 0.1859 1>,<at 0.5423 1>,<at 0.8523 0.5052>]}
      }}
     CurveIndex=$Particles.NormalizedAge
    )) {isSetParameters=true, module=SetVariables_#6}
  ]
  rendererClasses [&NiagaraMeshRendererProperties]
  renderers [
   {rendererClass=&NiagaraMeshRendererProperties, rendererIndex=0}
   {rendererClass=&NiagaraMeshRendererProperties, rendererIndex=1}
  ]
 }
 Cylinder {
  emitterUpdateScript [
   &EmitterState(
    Enable Distance Culling=true
    Inactive Response=::NewEnumerator1
    Life Cycle Mode=::NewEnumerator1
    Loop Duration=1
    MaxDistance=12500
    Module.bUseMaxDistance=true
    Module.bUseMinDistance=false
    Recalculate Duration Each Loop=true
    Scalability Mode=::NewEnumerator1
    UseLoopDelay=false
   )
   &SpawnBurst_Instantaneous(Spawn Probability=HlslExpression() {hlslExpression="Emitter.LoopCount == 0"}, Use Spawn Probability=true, UseLoopCountLimit=false)
  ]
  particleSpawnScript [
   &InitializeParticle(
    Color=&MakeLinearColorFromVectorAndFloat(
     "Vector (RGB)"=&Add_Vector(
      A=&MakeVectorFromLinearColorRGB(Color=$User.Variant_Disk_ColorA)
      B=&VectorFromFloat(Value=0.1)
     )
     "Float (Alpha)"=1
    )
    Mesh Scale=<1.4 1.4 1>
    Mesh Scale Mode=::NewEnumerator4
    UsePositionOffset=false
    Write Material Random=true
   )
   &InitialMeshOrientation(Mesh Axis To Orient=<0 0 0>, Mesh Orientation Mode=::NewEnumerator2, Orientation Vector=<0 0 1>, Use Rotation Vector=true)
  ]
  particleUpdateScript [
   &ParticleState(
    Kill Particles When Lifetime Has Elapsed=false
    Let Infinitely Lived Particles Die When Emitter Deactivates=true
    Loop Particles Lifetime=true
   )
   &SkeletalMeshLocation(MaskSpawnGroup=false, Mesh Sampling Type=::NewEnumerator5, Skeletal Mesh=$User.SkeletalMesh, Socket Sampling Mode=::NewEnumerator5)
   &ScaleColor(Scale Alpha=&Lerp_Float(A=1.13, B=0, Alpha=$System.Timeline.Time), ScaleA=true, ScaleRGB=true)
   &ScaleMeshSize(Scale Factor=&Lerp_Vector(
     A=<1.5 1.5 1.5>
     B=&Lerp_Vector(A=<1 1 1>, B=<0.9 0.9 0.9>, Alpha=$System.Timeline003.Time)
     Alpha=$System.Timeline002.Time
    ))
  ]
  rendererClasses [&NiagaraMeshRendererProperties]
  renderers [{rendererClass=&NiagaraMeshRendererProperties, rendererIndex=0}]
 }
 Cylinder001 {
  emitterUpdateScript [
   &EmitterState(
    Enable Distance Culling=true
    Inactive Response=::NewEnumerator1
    Life Cycle Mode=::NewEnumerator1
    Loop Duration=1
    MaxDistance=12500
    Module.bUseMaxDistance=true
    Module.bUseMinDistance=false
    Recalculate Duration Each Loop=true
    Scalability Mode=::NewEnumerator1
    UseLoopDelay=false
   )
   &SpawnBurst_Instantaneous(
    Spawn Count=5
    Spawn Probability=HlslExpression() {hlslExpression="Emitter.LoopCount == 0"}
    Use Spawn Probability=true
    UseLoopCountLimit=false
   )
  ]
  particleSpawnScript [
   &InitializeParticle(
    Color=$User.Variant_Disk_ColorA
    Lifetime=0.2
    Mesh Scale=&MakeVector(X=1.4, Y=1.4, Z=&RandomRangeFloat(Minimum=0.5, Maximum=1))
    Mesh Scale Mode=::NewEnumerator4
    UsePositionOffset=false
    Write Material Random=true
   )
   &DynamicMaterialParameters(
    Index 0 Param 2=&FloatFromCurve(
     FloatCurve=DataInterface() {propertyValues {Curve {keys [<at 0 0 RCIM_Cubic>, {arriveTangent=1, interpMode=RCIM_Cubic, leaveTangent=1, time=0.7035, value=0.3698}, <at 1 1>]}}}
     CurveIndex=&RandomRangeFloat(Minimum=0, Maximum=1)
    )
    Module.Param0WriteEnabled=true
    Module.Param1WriteEnabled=true
    Module.Param2WriteEnabled=true
    Module.Param3WriteEnabled=true
   ) {module=DynamicMaterialParameters001}
   &InitialMeshOrientation(
    Mesh Axis To Orient=<0 0 0>
    Mesh Orientation Mode=::NewEnumerator2
    Orientation Vector=<0 0 1>
    Rotation=&MakeVector(X=0, Y=0, Z=&RandomRangeFloat(Minimum=0, Maximum=1))
    Use Rotation Vector=true
   )
  ]
  particleUpdateScript [
   &ParticleState(
    Kill Particles When Lifetime Has Elapsed=false
    Let Infinitely Lived Particles Die When Emitter Deactivates=true
    Loop Particles Lifetime=true
   )
   &SkeletalMeshLocation(MaskSpawnGroup=false, Mesh Sampling Type=::NewEnumerator5, Skeletal Mesh=$User.SkeletalMesh, Socket Sampling Mode=::NewEnumerator5)
   &DynamicMaterialParameters(
    Index 0 Param 1=&FloatFromCurve(FloatCurve=DataInterface() {propertyValues {Curve {keys [<at 0 0>,<at 1 1>]}}})
    Module.Param0WriteEnabled=true
    Module.Param1WriteEnabled=false
    Module.Param2WriteEnabled=true
    Module.Param3WriteEnabled=true
   )
   &ScaleColor(Scale Alpha=&Lerp_Float(A=1, B=0, Alpha=$System.Timeline.Time), ScaleA=true, ScaleRGB=true)
   &ScaleMeshSize(Scale Factor=&Lerp_Vector(
     A=<1.5 1.5 1.5>
     B=&Lerp_Vector(A=<1 1 1>, B=<0.9 0.9 0.9>, Alpha=$System.Timeline003.Time)
     Alpha=$System.Timeline002.Time
    ))
  ]
  rendererClasses [&NiagaraMeshRendererProperties]
  renderers [{rendererClass=&NiagaraMeshRendererProperties, rendererIndex=0}]
 }
 Disk_Low {
  emitterSpawnScript [/SetVariables(Emitter.Low_Version=1) {isSetParameters=true}]
  emitterUpdateScript [
   &EmitterState(
    Inactive Response=::NewEnumerator1
    Life Cycle Mode=::NewEnumerator1
    Loop Duration=1
    Recalculate Duration Each Loop=true
    UseLoopDelay=false
   )
   &SpawnBurst_Instantaneous(Spawn Probability=HlslExpression() {hlslExpression="Emitter.LoopCount == 0"}, Use Spawn Probability=true, UseLoopCountLimit=false)
  ]
  particleSpawnScript [
   &InitializeParticle(Color Mode=::NewEnumerator0, Mesh Scale Mode=::NewEnumerator3, Mesh Uniform Scale=1.6, UsePositionOffset=false, Write Material Random=true)
   &InitialMeshOrientation(Mesh Axis To Orient=<0 0 0>, Mesh Orientation Mode=::NewEnumerator2, Orientation Vector=<0 0 0>, Use Rotation Vector=true) {enabled=false}
  ]
  particleUpdateScript [
   &ParticleState(
    Kill Particles When Lifetime Has Elapsed=false
    Let Infinitely Lived Particles Die When Emitter Deactivates=true
    Loop Particles Lifetime=true
   )
   &SkeletalMeshLocation(MaskSpawnGroup=false, Mesh Sampling Type=::NewEnumerator5, Skeletal Mesh=$User.SkeletalMesh, Socket Sampling Mode=::NewEnumerator5)
   &ScaleColor(Scale Alpha=&Lerp_Float(A=1, B=0, Alpha=$System.Timeline.Time), ScaleA=true, ScaleRGB=true)
   &ScaleMeshSize(Scale Factor=&Lerp_Vector(
     A=<1.5 1.5 1.5>
     B=&Lerp_Vector(A=<1 1 1>, B=<0.9 0.9 0.9>, Alpha=$System.Timeline003.Time)
     Alpha=$System.Timeline002.Time
    ))
   &CameraOffset(Camera Offset Amount=10) {enabled=false}
   &Material()
  ]
  rendererClasses [&NiagaraMeshRendererProperties]
  renderers [{rendererClass=&NiagaraMeshRendererProperties, rendererIndex=0}]
 }
 Disk_Sample {
  emitterSpawnScript [/SetVariables() {isSetParameters=true}]
  emitterUpdateScript [
   &EmitterState(
    Inactive Response=::NewEnumerator1
    Life Cycle Mode=::NewEnumerator1
    Loop Duration=1
    Recalculate Duration Each Loop=true
    UseLoopDelay=false
   )
   &SpawnBurst_Instantaneous(
    Spawn Count=8
    Spawn Probability=HlslExpression() {hlslExpression="Emitter.LoopCount == 0"}
    Use Spawn Probability=true
    UseLoopCountLimit=false
   )
  ]
  particleSpawnScript [
   &InitializeParticle(
    Color=$User.Variant_Disk_ColorA
    Mesh Scale Mode=::NewEnumerator3
    Mesh Uniform Scale=1.6
    UsePositionOffset=false
    Write Material Random=true
   )
   &InitialMeshOrientation(Mesh Axis To Orient=<0 0 0>, Mesh Orientation Mode=::NewEnumerator2, Orientation Vector=<0 0 0>, Use Rotation Vector=true) {enabled=false}
   &FN_SetSpawnGroup(Number of Spawn Groups=8)
  ]
  particleUpdateScript [
   &ParticleState(
    Kill Particles When Lifetime Has Elapsed=false
    Let Infinitely Lived Particles Die When Emitter Deactivates=true
    Loop Particles Lifetime=true
   )
   &SkeletalMeshLocation(Bone / Socket Index=$Particles.SpawnGroup, Bone Sampling Mode=::NewEnumerator4, MaskSpawnGroup=false, Skeletal Mesh=$User.SkeletalMesh)
   &ScaleColor(
    Scale Alpha=&Lerp_Float(A=1, B=0, Alpha=$System.Timeline.Time)
    Scale RGB=<0.8 0.8 0.8>
    ScaleA=true
    ScaleRGB=true
   )
   &ScaleMeshSize(Scale Factor=&Lerp_Vector(
     A=<1.5 1.5 1.5>
     B=&Lerp_Vector(A=<1 1 1>, B=<0.9 0.9 0.9>, Alpha=$System.Timeline003.Time)
     Alpha=$System.Timeline002.Time
    ))
   &CameraOffset(Camera Offset Amount=10) {enabled=false}
   &Material()
  ]
  rendererClasses [&NiagaraMeshRendererProperties]
  renderers [{rendererClass=&NiagaraMeshRendererProperties, rendererIndex=0}]
 }
 GlowDrops {
  emitterUpdateScript [
   &EmitterState(
    Enable Distance Culling=true
    Max Distance Response=::KillImmediately
    Module.bApplySpawnCountScaleByDistance=true
    Module.bUseMaxDistance=true
    Module.bUseMinDistance=false
    Scalability Mode=::NewEnumerator1
    Scale Spawn Count=true
    Spawn Count Scale=3
    Spawn Count Scale By Distance=DataInterface() {propertyValues {Curve {keys [<at 0.3676 0.9657>,<at 0.9962 0.4982>]}}}
   )
   &SpawnBurst_Instantaneous(
    Spawn Count=&UniformRangedInt(Minimum=0, Maximum=1)
    Spawn Time=&RandomRangeFloat(Minimum=0, Maximum=1)
    Use Spawn Probability=false
    UseLoopCountLimit=false
   )
   &SpawnRate(
    SpawnRate=&Lerp_Float(
     A=2
     B=&Lerp_Float(A=&RandomRangeFloat(Minimum=1, Maximum=2), B=2, Alpha=$System.Timeline003.Time)
     Alpha=$System.Timeline002.Time
    )
    Use Spawn Probability=true
   )
  ]
  particleSpawnScript [
   &InitializeParticle(
    Color=&MultiplyLinearColor(Linear Color=$User.Variant_Color, Scale Factor=<10 10 10 2>)
    Lifetime Max=0.5
    Lifetime Min=0.25
    Lifetime Mode=::NewEnumerator1
    Sprite Size Mode=::NewEnumerator1
    Uniform Sprite Size Max=5
    Uniform Sprite Size Min=2.5
    UsePositionOffset=false
    Write Material Random=true
   )
   &AddVelocityFromPoint(Module.Use Falloff Distance=false, Velocity Strength=&UniformRangedFloat(Minimum=500, Maximum=1000))
   &ShapeLocation(
    Cylinder Distribution=::NewEnumerator2
    Cylinder Height=1
    MaskSpawnGroup=false
    Offset=<-100 25 0>
    Offset Mode=::NewEnumerator0
    Shape Primitive=::NewEnumerator1
    U Position=&Multiply_Float(A=1, B=1)
    UseOwnerScale=true
   ) {module=ShapeLocation001}
  ]
  particleUpdateScript [
   &ParticleState()
   &GravityForce()
   &Drag(Drag=&UniformRangedFloat(Minimum=5, Maximum=15), Use Linear Drag=true, Use Rotational Drag=false)
   &InheritSourceMovement(
    Applied Force Scale=<-10 -10 0>
    Applied Force Speed Limit=10
    ApplyForce=true
    ApplyForceFalloff=false
    ApplyPosition=false
    Module.LimitForceSpeed=true
    Source Velocity:Vector3f=&TransformVector(Vector=$Engine.Owner.Velocity, Source Space=::World, Destination Space=::Simulation)
   )
   &CurlNoiseForce(Noise Strength=&RandomRangeFloat(Minimum=200, Maximum=400), Pan Noise=true)
   &Color(
    Color=$User.Variant_Color
    Scale Alpha=&FloatFromCurve(FloatCurve=DataInterface() {propertyValues {Curve {keys [{arriveTangent=-2.6501, interpMode=RCIM_Cubic, leaveTangent=-2.6501, tangentMode=RCTM_User, time=0, value=1}, <at 1 0>]}}})
    Scale Color=&VectorFromCurve(
     VectorCurve=DataInterface() {propertyValues {
       XCurve {keys [
         {tangentMode=RCTM_User, time=0, value=2}
         {arriveTangent=-0.0405, interpMode=RCIM_Cubic, leaveTangent=-0.0405, tangentMode=RCTM_User, time=0.7009, value=2}
         {arriveTangent=-8.0089, interpMode=RCIM_Cubic, leaveTangent=-8.0089, tangentMode=RCTM_User, time=1, value=0}
        ]}
       YCurve {keys [
         {tangentMode=RCTM_User, time=0, value=1}
         {arriveTangent=-0.0405, interpMode=RCIM_Cubic, leaveTangent=-0.0405, tangentMode=RCTM_User, time=0.7009, value=1}
         {arriveTangent=-8.0089, interpMode=RCIM_Cubic, leaveTangent=-8.0089, tangentMode=RCTM_User, time=1, value=0}
        ]}
       ZCurve {keys [
         {tangentMode=RCTM_User, time=0, value=1}
         {arriveTangent=-0.0405, interpMode=RCIM_Cubic, leaveTangent=-0.0405, tangentMode=RCTM_User, time=0.7009, value=1}
         {arriveTangent=-8.0089, interpMode=RCIM_Cubic, leaveTangent=-8.0089, tangentMode=RCTM_User, time=1, value=0}
        ]}
      }}
     Scale Curve=<5 5 5>
    )
   )
   &InheritVelocity(Inherited Velocity Amount Scale=<0.5 0.5 0.1>, Speed Limit=true, UseSpeedThreshold=false)
   &SolveForcesAndVelocity(Clamp Velocity=false, Limit Acceleration=false)
   &ScaleSpriteSize(Scale Factor=&Vector2DFromCurve(
     Vector2Curve=DataInterface() {propertyValues {
       XCurve {keys [<at 0 1 RCIM_Cubic>, {arriveTangent=-2.8959, interpMode=RCIM_Cubic, leaveTangent=-2.8959, tangentMode=RCTM_User, time=1, value=0}]}
       YCurve {keys [<at 0 1 RCIM_Cubic>, {arriveTangent=-2.8959, interpMode=RCIM_Cubic, leaveTangent=-2.8959, tangentMode=RCTM_User, time=1, value=0}]}
      }}
     ScaleCurve=<1 1>
    ))
   &ScaleSpriteSizeBySpeed(Max Scale Factor=<0.5 10>, Min Scale Factor=<0.8 1.2>)
   &ScaleColor(
    Scale Alpha=&Lerp_Float(A=&FN_FrontEndFloatModifier(Gameplay=1, Front End=0), B=0, Alpha=$System.Timeline.Time)
    Scale RGB=&MakeVectorFromLinearColorRGB(Color=$User.Variant_Color)
    ScaleA=true
    ScaleRGB=true
   ) {module=ScaleColor001}
  ]
  rendererClasses [&NiagaraSpriteRendererProperties]
  renderers [{rendererClass=&NiagaraSpriteRendererProperties}]
 }
 GlowDrops001 {
  emitterUpdateScript [
   &EmitterState(
    Enable Distance Culling=true
    Max Distance Response=::KillImmediately
    Module.bApplySpawnCountScaleByDistance=true
    Module.bUseMaxDistance=true
    Module.bUseMinDistance=false
    Scalability Mode=::NewEnumerator1
    Scale Spawn Count=true
    Spawn Count Scale=3
    Spawn Count Scale By Distance=DataInterface() {propertyValues {Curve {keys [<at 0.3676 0.9657>,<at 0.9962 0.4982>]}}}
   )
   &SpawnBurst_Instantaneous(
    Spawn Count=&UniformRangedInt(Minimum=0, Maximum=1)
    Spawn Time=&RandomRangeFloat(Minimum=0, Maximum=1)
    Use Spawn Probability=false
    UseLoopCountLimit=false
   )
   &SpawnRate(
    Spawn Probability=0.5
    SpawnRate=&Lerp_Float(
     A=2
     B=&Lerp_Float(A=&RandomRangeFloat(Minimum=1, Maximum=2), B=2, Alpha=$System.Timeline003.Time)
     Alpha=$System.Timeline002.Time
    )
    Use Spawn Probability=true
   )
  ]
  particleSpawnScript [
   &InitializeParticle(
    Color=&MultiplyLinearColor(Linear Color=$User.Variant_Color, Scale Factor=<10 10 10 2>)
    Lifetime Max=0.5
    Lifetime Min=0.25
    Lifetime Mode=::NewEnumerator1
    Sprite Size Mode=::NewEnumerator1
    Uniform Sprite Size Max=5
    Uniform Sprite Size Min=2.5
    UsePositionOffset=false
    Write Material Random=true
   )
   &AddVelocityFromPoint(Module.Use Falloff Distance=false, Velocity Strength=&UniformRangedFloat(Minimum=500, Maximum=1000))
   &ShapeLocation(
    Cylinder Distribution=::NewEnumerator2
    Cylinder Height=1
    MaskSpawnGroup=false
    Offset=<-100 -25 0>
    Offset Mode=::NewEnumerator0
    Shape Primitive=::NewEnumerator1
    U Position=&Multiply_Float(A=1, B=1)
    UseOwnerScale=true
   ) {module=ShapeLocation001}
  ]
  particleUpdateScript [
   &ParticleState()
   &GravityForce()
   &Drag(Drag=&UniformRangedFloat(Minimum=5, Maximum=15), Use Linear Drag=true, Use Rotational Drag=false)
   &InheritSourceMovement(
    Applied Force Scale=<-10 -10 0>
    Applied Force Speed Limit=10
    ApplyForce=true
    ApplyForceFalloff=false
    ApplyPosition=false
    Module.LimitForceSpeed=true
    Source Velocity:Vector3f=&TransformVector(Vector=$Engine.Owner.Velocity, Source Space=::World, Destination Space=::Simulation)
   )
   &CurlNoiseForce(Noise Strength=&RandomRangeFloat(Minimum=200, Maximum=400), Pan Noise=true)
   &Color(
    Color=$User.Variant_Color
    Scale Alpha=&FloatFromCurve(FloatCurve=DataInterface() {propertyValues {Curve {keys [{arriveTangent=-2.6501, interpMode=RCIM_Cubic, leaveTangent=-2.6501, tangentMode=RCTM_User, time=0, value=1}, <at 1 0>]}}})
    Scale Color=&VectorFromCurve(
     VectorCurve=DataInterface() {propertyValues {
       XCurve {keys [
         {tangentMode=RCTM_User, time=0, value=2}
         {arriveTangent=-0.0405, interpMode=RCIM_Cubic, leaveTangent=-0.0405, tangentMode=RCTM_User, time=0.7009, value=2}
         {arriveTangent=-8.0089, interpMode=RCIM_Cubic, leaveTangent=-8.0089, tangentMode=RCTM_User, time=1, value=0}
        ]}
       YCurve {keys [
         {tangentMode=RCTM_User, time=0, value=1}
         {arriveTangent=-0.0405, interpMode=RCIM_Cubic, leaveTangent=-0.0405, tangentMode=RCTM_User, time=0.7009, value=1}
         {arriveTangent=-8.0089, interpMode=RCIM_Cubic, leaveTangent=-8.0089, tangentMode=RCTM_User, time=1, value=0}
        ]}
       ZCurve {keys [
         {tangentMode=RCTM_User, time=0, value=1}
         {arriveTangent=-0.0405, interpMode=RCIM_Cubic, leaveTangent=-0.0405, tangentMode=RCTM_User, time=0.7009, value=1}
         {arriveTangent=-8.0089, interpMode=RCIM_Cubic, leaveTangent=-8.0089, tangentMode=RCTM_User, time=1, value=0}
        ]}
      }}
     Scale Curve=<5 5 5>
    )
   )
   &InheritVelocity(Inherited Velocity Amount Scale=<0.5 0.5 0.1>, Speed Limit=true, UseSpeedThreshold=false)
   &SolveForcesAndVelocity(Clamp Velocity=false, Limit Acceleration=false)
   &ScaleSpriteSize(Scale Factor=&Vector2DFromCurve(
     Vector2Curve=DataInterface() {propertyValues {
       XCurve {keys [<at 0 1 RCIM_Cubic>, {arriveTangent=-2.8959, interpMode=RCIM_Cubic, leaveTangent=-2.8959, tangentMode=RCTM_User, time=1, value=0}]}
       YCurve {keys [<at 0 1 RCIM_Cubic>, {arriveTangent=-2.8959, interpMode=RCIM_Cubic, leaveTangent=-2.8959, tangentMode=RCTM_User, time=1, value=0}]}
      }}
     ScaleCurve=<1 1>
    ))
   &ScaleSpriteSizeBySpeed(Max Scale Factor=<0.5 10>, Min Scale Factor=<0.8 1.2>)
   &ScaleColor(
    Scale Alpha=&Lerp_Float(A=&FN_FrontEndFloatModifier(Gameplay=1, Front End=0), B=0, Alpha=$System.Timeline.Time)
    Scale RGB=&MakeVectorFromLinearColorRGB(Color=$User.Variant_Color)
    ScaleA=true
    ScaleRGB=true
   ) {module=ScaleColor002}
  ]
  rendererClasses [&NiagaraSpriteRendererProperties]
  renderers [{rendererClass=&NiagaraSpriteRendererProperties}]
 }
 Light {
  emitterSpawnScript [&FN_PreserveSkydiveUserVariables() {module=PreserveUserVariables}]
  emitterUpdateScript [
   &EmitterState(
    Enable Distance Culling=true
    Life Cycle Mode=::NewEnumerator1
    Loop Behavior=::NewEnumerator1
    Loop Duration Mode=::NewEnumerator1
    Module.bUseMaxDistance=true
    Module.bUseMinDistance=false
    Scalability Mode=::NewEnumerator1
    UseLoopDelay=false
   )
   &SpawnBurst_Instantaneous(Spawn Count=1, Spawn Probability=HlslExpression() {hlslExpression="Emitter.LoopCount == 0"}, Use Spawn Probability=true)
  ]
  particleSpawnScript [
   /Niagara/Modules/Spawn/Initialization/InitializeParticle(
    Color=$User.Variant_Light_Color
    Module.Write Color=true
    Module.Write Lifetime=true
    Module.Write Mass=false
    Module.Write Material Random=true
    Module.Write Position=false
    Module.Write Scale=false
    Module.Write SpriteRotation=false
    Module.Write SpriteSize=false
   )
   &CameraOffset(Camera Offset Amount=15)
  ]
  particleUpdateScript [
   &ParticleState(
    Kill Particles When Lifetime Has Elapsed=false
    Let Infinitely Lived Particles Die When Emitter Deactivates=true
    Loop Particles Lifetime=true
   )
   &SkeletalMeshLocation(MaskSpawnGroup=false, Mesh Sampling Type=::NewEnumerator5, Skeletal Mesh=$User.SkeletalMesh, Socket Sampling Mode=::NewEnumerator5)
   &FN_Light_Attributes(
    Color=$User.Variant_Light_Color
    Create Color Attribute=true
    Scale Color=&Lerp_Float(A=1, B=0, Alpha=$System.Timeline.Time)
   )
  ]
  rendererClasses [&NiagaraLightRendererProperties]
  renderers [{rendererClass=&NiagaraLightRendererProperties, rendererIndex=0}]
 }
 Sparks {
  emitterUpdateScript [
   &EmitterState(
    Enable Distance Culling=true
    MaxDistance=10000
    Module.bUseMaxDistance=true
    Module.bUseMinDistance=false
    Scalability Mode=::NewEnumerator1
   )
   &SpawnBurst_Instantaneous(
    Spawn Count=&RandomRangeInt(Minimum=2, Maximum=3)
    Spawn Probability=&Lerp_Float(A=1, B=0, Alpha=$System.Timeline.Time)
    Spawn Time=&RandomRangeFloat(Minimum=0, Maximum=1)
    Use Spawn Probability=true
   )
   &SpawnBurst_Instantaneous(
    Spawn Count=&RandomRangeInt(Minimum=5, Maximum=7)
    Spawn Probability=&Lerp_Float(A=1, B=0, Alpha=$System.Timeline.Time)
    Spawn Time=&RandomRangeFloat(Minimum=0, Maximum=1)
    Use Spawn Probability=true
   ) {module=SpawnBurst_Instantaneous001}
   &SpawnRate(
    SpawnRate=&Lerp_Float(
     A=&Lerp_Float(A=50, B=&Lerp_Float(A=0, B=50, Alpha=$System.Timeline003.Time), Alpha=$System.Timeline002.Time)
     B=0
     Alpha=$System.Timeline.Time
    )
    Use Spawn Probability=false
   )
  ]
  particleSpawnScript [
   /SetVariables(Particles.Random=&RandomRangeFloat(Minimum=0, Maximum=1)) {isSetParameters=true, module=SetVariables_#7}
   &InitializeParticle(
    Color=$User.Variant_Color
    Lifetime=&Multiply_Float(A=&Lerp_Float(A=0.1, B=0.3, Alpha=$Particles.Random), B=0.7)
    Mesh Scale=&Multiply_Vector(
     A=&Multiply_Vector(A=&Lerp_Vector(A=<0.2 0.2 0.2>, B=<0.6 0.6 0.6>, Alpha=$Particles.Random), B=<1 1 0.4>)
     B=&VectorFromFloat(Value=0.65)
    )
    Module.Write Color=true
    Module.Write Lifetime=true
    Module.Write Mass=true
    Module.Write Material Random=true
    Module.Write Position=false
    Module.Write Scale=true
    Module.Write SpriteRotation=false
    Module.Write SpriteSize=false
    Position=&SimulationPosition()
   )
   &CameraOffset(Camera Offset Amount=30)
   &ShapeLocation(
    Cylinder Distribution=::NewEnumerator2
    Cylinder Height=1
    MaskSpawnGroup=false
    Offset=<-100 -25 0>
    Offset Mode=::NewEnumerator0
    Shape Primitive=::NewEnumerator1
    U Position=&Multiply_Float(A=1, B=1)
    UseOwnerScale=true
   ) {module=ShapeLocation001}
   &DynamicMaterialParameters(
    Index 0 Param 2=&FloatFromCurve(
     FloatCurve=DataInterface() {propertyValues {Curve {keys [<at 0 0 RCIM_Cubic>, {arriveTangent=1, interpMode=RCIM_Cubic, leaveTangent=1, time=0.7035, value=0.3698}, <at 1 1>]}}}
     CurveIndex=&RandomRangeFloat(Minimum=0, Maximum=1)
    )
    Module.Param0WriteEnabled=true
    Module.Param1WriteEnabled=true
    Module.Param2WriteEnabled=true
    Module.Param3WriteEnabled=true
   ) {module=DynamicMaterialParameters001}
   &InitialMeshOrientation(
    Orientation Vector=<1 1 1>
    Rotation=&RandomRangeVector(Minimum=<-1 -1 -1>, Maximum=<1 1 1>)
    Use Orientation Vector=true
    Use Rotation Vector=true
   )
  ]
  particleUpdateScript [
   &ParticleState()
   &SolveForcesAndVelocity(Clamp Velocity=false, Limit Acceleration=false)
   &DynamicMaterialParameters(
    Index 0 Param 1=&FloatFromCurve(FloatCurve=DataInterface() {propertyValues {Curve {keys [<at 0 0>,<at 1 1>]}}})
    Module.Param0WriteEnabled=true
    Module.Param1WriteEnabled=false
    Module.Param2WriteEnabled=true
    Module.Param3WriteEnabled=true
   )
   &ScaleColor(
    Scale Alpha=&Lerp_Float(A=&FN_FrontEndFloatModifier(Gameplay=1, Front End=0), B=0, Alpha=$System.Timeline.Time)
    ScaleA=true
    ScaleRGB=true
   )
  ]
  rendererClasses [&NiagaraMeshRendererProperties]
  renderers [{rendererClass=&NiagaraMeshRendererProperties, rendererIndex=0}]
 }
 Sparks001 {
  emitterUpdateScript [
   &EmitterState(
    Enable Distance Culling=true
    MaxDistance=10000
    Module.bUseMaxDistance=true
    Module.bUseMinDistance=false
    Scalability Mode=::NewEnumerator1
   )
   &SpawnBurst_Instantaneous(
    Spawn Count=&RandomRangeInt(Minimum=2, Maximum=3)
    Spawn Probability=&Lerp_Float(A=1, B=0, Alpha=$System.Timeline.Time)
    Spawn Time=&RandomRangeFloat(Minimum=0, Maximum=1)
    Use Spawn Probability=true
   )
   &SpawnBurst_Instantaneous(
    Spawn Count=&RandomRangeInt(Minimum=5, Maximum=7)
    Spawn Probability=&Lerp_Float(A=1, B=0, Alpha=$System.Timeline.Time)
    Spawn Time=&RandomRangeFloat(Minimum=0, Maximum=1)
    Use Spawn Probability=true
   ) {module=SpawnBurst_Instantaneous001}
   &SpawnRate(
    SpawnRate=&Lerp_Float(
     A=&Lerp_Float(A=50, B=&Lerp_Float(A=0, B=50, Alpha=$System.Timeline003.Time), Alpha=$System.Timeline002.Time)
     B=0
     Alpha=$System.Timeline.Time
    )
    Use Spawn Probability=false
   )
  ]
  particleSpawnScript [
   /SetVariables(Particles.Random=&RandomRangeFloat(Minimum=0, Maximum=1)) {isSetParameters=true, module=SetVariables_#7}
   &InitializeParticle(
    Color=$User.Variant_Color
    Lifetime=&Multiply_Float(A=&Lerp_Float(A=0.1, B=0.3, Alpha=$Particles.Random), B=0.7)
    Mesh Scale=&Multiply_Vector(
     A=&Multiply_Vector(A=&Lerp_Vector(A=<0.2 0.2 0.2>, B=<0.6 0.6 0.6>, Alpha=$Particles.Random), B=<1 1 0.4>)
     B=&VectorFromFloat(Value=0.65)
    )
    Module.Write Color=true
    Module.Write Lifetime=true
    Module.Write Mass=true
    Module.Write Material Random=true
    Module.Write Position=false
    Module.Write Scale=true
    Module.Write SpriteRotation=false
    Module.Write SpriteSize=true
    Position=&SimulationPosition()
   )
   &CameraOffset(Camera Offset Amount=30)
   &ShapeLocation(
    Cylinder Distribution=::NewEnumerator2
    Cylinder Height=1
    MaskSpawnGroup=false
    Offset=&FN_FrontEndVector3Modifier(Gameplay=<-100 25 0>, Front End=<0 0 0>)
    Offset Mode=::NewEnumerator0
    Shape Primitive=::NewEnumerator1
    U Position=&Multiply_Float(A=1, B=1)
    UseOwnerScale=true
   ) {module=ShapeLocation001}
   &DynamicMaterialParameters(
    Index 0 Param 2=&FloatFromCurve(
     FloatCurve=DataInterface() {propertyValues {Curve {keys [<at 0 0 RCIM_Cubic>, {arriveTangent=1, interpMode=RCIM_Cubic, leaveTangent=1, time=0.7035, value=0.3698}, <at 1 1>]}}}
     CurveIndex=&RandomRangeFloat(Minimum=0, Maximum=1)
    )
    Module.Param0WriteEnabled=true
    Module.Param1WriteEnabled=true
    Module.Param2WriteEnabled=true
    Module.Param3WriteEnabled=true
   ) {module=DynamicMaterialParameters001}
   &InitialMeshOrientation(
    Orientation Vector=<1 1 1>
    Rotation=&RandomRangeVector(Minimum=<-1 -1 -1>, Maximum=<1 1 1>)
    Use Orientation Vector=true
    Use Rotation Vector=true
   )
  ]
  particleUpdateScript [
   &ParticleState()
   &SolveForcesAndVelocity(Clamp Velocity=false, Limit Acceleration=false)
   &DynamicMaterialParameters(
    Index 0 Param 1=&FloatFromCurve(FloatCurve=DataInterface() {propertyValues {Curve {keys [<at 0 0>,<at 1 1>]}}})
    Module.Param0WriteEnabled=true
    Module.Param1WriteEnabled=false
    Module.Param2WriteEnabled=true
    Module.Param3WriteEnabled=true
   )
   &ScaleColor(
    Scale Alpha=&Lerp_Float(A=&FN_FrontEndFloatModifier(Gameplay=1, Front End=0), B=0, Alpha=$System.Timeline.Time)
    ScaleA=true
    ScaleRGB=true
   )
  ]
  rendererClasses [&NiagaraMeshRendererProperties]
  renderers [{rendererClass=&NiagaraMeshRendererProperties, rendererIndex=0}]
 }
}
summary {
 systemName=NS_Glider_KigntCat_Idle
 emitters [
  {emitterName=Light, rendererClasses [&NiagaraLightRendererProperties]}
  {emitterName=Disk_Low, rendererClasses [&NiagaraMeshRendererProperties]}
  {emitterName=Cylinder, rendererClasses [&NiagaraMeshRendererProperties]}
  {emitterName=Sparks, rendererClasses [&NiagaraMeshRendererProperties]}
  {emitterName=GlowDrops001, rendererClasses [&NiagaraSpriteRendererProperties]}
  {emitterName=GlowDrops, rendererClasses [&NiagaraSpriteRendererProperties]}
  {emitterName=Sparks001, rendererClasses [&NiagaraMeshRendererProperties]}
  {emitterName=Cylinder001, rendererClasses [&NiagaraMeshRendererProperties]}
  {emitterName=Disk_Sample, rendererClasses [&NiagaraMeshRendererProperties]}
  {emitterName=Blades_Ghost_Trail005, rendererClasses [&NiagaraMeshRendererProperties]}
  {emitterName=Blades_Ghost_Trail004, rendererClasses [&NiagaraMeshRendererProperties]}
  {emitterName=Blades_Ghost_Trail006, rendererClasses [&NiagaraMeshRendererProperties]}
  {emitterName=Blades_Ghost_Trail007, rendererClasses [&NiagaraMeshRendererProperties]}
 ]
 userVariables [
  User.AlphaMult {type=NiagaraFloat, defaultValue=0, description=""}
  User.bIsFrontEnd {type=NiagaraBool}
  User.bIsFrontEndPreview {type=NiagaraBool}
  User.bIsFullyDeployed {type=NiagaraBool}
  User.CorrectDirection {type=NiagaraFloat, defaultValue=0, description=""}
  User.ForwardDot {type=NiagaraFloat, defaultValue=0, description=""}
  User.FrontendSpawn {type=NiagaraBool}
  User.PlayerSpeed {type=NiagaraFloat, defaultValue=0, description=""}
  User.RibbonIDOffset {type=NiagaraInt32}
  User.RightDot {type=NiagaraFloat, defaultValue=0, description=""}
  User.SkeletalMesh {
   type=NiagaraDataInterfaceSkeletalMesh
   defaultValue {
    struct=NiagaraExt_VariableValue_DataInterface
    value {dataInterface=&self:NiagaraDataInterfaceSkeletalMesh_1}
   }
  }
  User.SplineUpdatedEndPos {type=Vector3f, description="", defaultValue {struct=Vector3f, value=<0 0 0>}}
  User.Variant_Color {type=LinearColor, description="", defaultValue {struct=LinearColor, value=<rgba 1 0 1 1>}}
  User.Variant_Disk_ColorA {type=LinearColor, description="", defaultValue {struct=LinearColor, value=<rgba 1 1 1 1>}}
  User.Variant_Disk_ColorB {type=LinearColor, description="", defaultValue {struct=LinearColor, value=<rgba 1 0 0.5 1>}}
  User.Variant_Disk_ColorC {type=LinearColor, description="", defaultValue {struct=LinearColor, value=<rgba 1 0 0.2 1>}}
  User.Variant_Disk_ColorD {type=LinearColor, description="", defaultValue {struct=LinearColor, value=<rgba 1 1 1 1>}}
  User.Variant_Light_Color {type=LinearColor, description="", defaultValue {struct=LinearColor, value=<rgba 2 0.05 0.5 1>}}
 ]
}
